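# Unlit, two-sided translucent glow material for Inferno pickaxe effects on Niagara sprites and mesh particles.
# A diamond gradient or soft UV sphere mask, cut by an inverted local-position sphere whose center comes from a dynamic parameter, blends two colors and drives opacity.
# Emissive adds a UV-based blade boost and particle color tint.
# Opacity also uses particle alpha and view-angle fade, plus depth fade except on mobile.
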Material /BRCosmetics/Weapons/FORT_Melee/Pickaxe_Inferno/FX/MI_Pickaxe_Inferno_Glow_01
graph {
 properties {
  bUsedWithNiagaraMeshParticles=true
  bUsedWithNiagaraSprites=true
  blendMode=BLEND_Translucent
  pixelDepthOffsetMode=PDOM_Legacy
  shadingModel=MSM_Unlit
  twoSided=true
  usedShadingModels=Unlit
 }
 nodes [
  n1 = ParticleColor()
  n2 = ComponentMask(None=TextureCoordinate()) {g=true}
  n3 = ScalarParameter() {defaultValue=4, parameterName=Blade Emissive Enhance}
  n4 = Multiply(
   A=MaterialFunctionCall!(Falloff=ScalarParameter() {defaultValue=1.6, parameterName=Diamond Power}) {materialFunction=/Engine/Functions/Engine_MaterialFunctions01/Gradient/DiamondGradient}.DiamondGradient
   B=ScalarParameter() {defaultValue=1, parameterName=Diamond Mult}
  )
  n5 = SphereMask(A=TextureCoordinate(), B=Constant() {r=0.5}, Radius=ScalarParameter() {defaultValue=0.4, parameterName=Sphere Radius}) {attenuationRadius=0.5, hardnessPercent=0}
  n6 = ScalarParameter() {defaultValue=3, parameterName=Sphere Power}
  n7 = DynamicParameter() {parameterIndex=3, paramNames [X,Y,Z,Param4]}
  MaterialExpressionAppend3Vector_0 = MaterialXAppend3Vector!(A=@n7.X, B=@n7.X, C=@n7.X)
  n8 = SphereMask(
   A=MaterialFunctionCall() {materialFunction=&LocalPosition}.Local Position
   B=@MaterialExpressionAppend3Vector_0
   Radius=ScalarParameter() {defaultValue=18, group=SpheraMask, parameterName=Local Radius}
   Hardness=ScalarParameter() {defaultValue=0.5, group=SpheraMask, parameterName=Local Hardness}
  )
  n9 = Multiply(
   A=LinearInterpolate(A=Saturate(None=@n4), B=Power(Base=@n5, Exp=@n6), Alpha=ScalarParameter() {parameterName=Lerp Sphere})
   B=LinearInterpolate(A=Constant() {r=1}, B=OneMinus(None=@n8), Alpha=ScalarParameter() {defaultValue=1, group=SpheraMask, parameterName=Use SpheraMask}) {constA=1}
  )
  n10 = LinearInterpolate(
   A=VectorParameter() {parameterName=Color A, defaultValue=<rgba 0.2333 0 0.7 1>}.RGB
   B=VectorParameter() {parameterName=Color B, defaultValue=<rgba 0 0 1 1>}.RGB
   Alpha=@n9
  )
  n11 = ScalarParameter() {defaultValue=1, parameterName=Emissive Mult}
  n12 = ParticleColor()
  n13 = Multiply(
   A=LinearInterpolate(B=Multiply(A=@n2, B=@n3) {constB=4}, Alpha=ScalarParameter() {defaultValue=0.5, parameterName=Lerp Blade Emissive Enhance}) {constA=1}
   B=Multiply(A=Multiply(A=@n10, B=@n11), B=@n12.RGB)
  )
  MaterialFunctionCall(Emissive=StaticSwitchParameter(True=Multiply(A=@n1.RGB, B=@n13), False=@n13) {defaultValue=true, parameterName=Use ParticleColor}) {materialFunction=&MF_FX_EmissiveAdjustments}
  n14 = Multiply(
   A=Multiply(A=Multiply(A=@n12.A, B=@n9), B=ScalarParameter() {defaultValue=1.5, parameterName=Opacity Mult})
   B=MaterialFunctionCall("FadeSensitivity (optional)"=ScalarParameter() {defaultValue=1, group=Alpha, parameterName=FadeViewAngle}) {materialFunction=&FadeBasedOnViewAngle}.Result
  )
  n15 = ScalarParameter() {defaultValue=10, group=Alpha, parameterName=FadeDistance}
  n16 = ShadingPathSwitch(Default=Multiply(A=@n14, B=DepthFade(FadeDistance=@n15)), Mobile=@n14) {desc=Migrated from FeatureLevelSwitch}
  StaticSwitchParameter(True=Multiply(A=@n1.A, B=@n16), False=@n16) {defaultValue=true, parameterName=Use ParticleColor}
 ]
}
