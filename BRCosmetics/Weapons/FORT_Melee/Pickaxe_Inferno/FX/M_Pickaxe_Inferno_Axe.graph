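# Translucent unlit material for Inferno pickaxe axe effects on Niagara mesh particles.
# Emissive maps UV gradients through curve-atlas color ramps, with optional panning highlights, triplanar galaxy stars and local-space halos, tinted by particle color.
# Opacity combines UV and local sphere masks, sine-based procedural shape fades and distorted panning noise with particle alpha.
# An optional switch fakes backface culling by comparing pixel depth against custom depth.
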Material /BRCosmetics/Weapons/FORT_Melee/Pickaxe_Inferno/FX/M_Pickaxe_Inferno_Axe
graph {
 properties {
  allowTranslucentCustomDepthWrites=true
  bUsedWithNiagaraMeshParticles=true
  blendMode=BLEND_Translucent
  opacityMaskClipValue=0.99
  pixelDepthOffsetMode=PDOM_Legacy
  shadingModel=MSM_Unlit
  usedShadingModels=Unlit
 }
 nodes [
  n1 = ParticleColor()
  n2 = ComponentMask(None=TextureCoordinate()) {g=true}
  n3 = CurveAtlasRowParameter(CurveTime=@n2) {
   atlas=/Game/Athena/LinearColorCurves/A_Atlas_Turtleneck_Glider
   curve=/Game/Weapons/FORT_Melee/Pickaxe_Inferno/FX/Curve_Pickaxe_Inferno_Color_01
   defaultValue=53
   parameterName=Curve_Color_01
  }
  n4 = ScalarParameter() {group=Color, parameterName=Emissive Dark}
  n5 = ScalarParameter() {defaultValue=1, group=Color, parameterName=Emissive Light}
  n6 = TextureCoordinate()
  n7 = CurveAtlasRowParameter(CurveTime=ComponentMask(None=@n6) {r=true}) {
   atlas=&A_Atlas_Turtleneck_Glider
   curve=/Game/Weapons/FORT_Melee/Pickaxe_Inferno/FX/Curve_Pickaxe_Inferno_Gradient_01
   defaultValue=52
   parameterName=Curve_Gradient_01
  }
  n8 = CurveAtlasRowParameter(CurveTime=ComponentMask(None=@n6) {g=true}) {
   atlas=&A_Atlas_Turtleneck_Glider
   curve=&Curve_Pickaxe_Inferno_Gradient_01
   defaultValue=52
   parameterName=Curve_Gradient_01
  }
  n9 = Multiply(A=@n7.R, B=@n8.G)
  n10 = CurveAtlasRowParameter(CurveTime=@n9) {
   atlas=&A_Atlas_Turtleneck_Glider
   curve=&Curve_Pickaxe_Inferno_Color_01
   defaultValue=53
   parameterName=Curve_Color_01
  }
  n11 = LinearInterpolate(
   A=VectorParameter() {group=Color, parameterName=Color Dark, defaultValue=<rgba 0 0 0 1>}.RGB
   B=VectorParameter() {group=Color, parameterName=Color Light, defaultValue=<rgba 1 1 1 1>}.RGB
   Alpha=@n9
  )
  n12 = VectorParameter() {group=Color, parameterName=Blade Color, defaultValue=<rgba 0.6833 0.05 1 1>}
  n13 = Multiply(A=Add(A=@n7.A, B=@n8.B), B=@n8.B)
  n14 = Multiply(
   A=StaticSwitchParameter(True=@n3.A, False=LinearInterpolate(A=@n4, B=@n5, Alpha=@n2)) {defaultValue=true, group=Color, parameterName=Use Color Curve}
   B=Add(
    A=StaticSwitchParameter(True=@n10, False=@n11) {defaultValue=true, group=Color, parameterName=Use Color Curve}
    B=Multiply(A=@n12.RGB, B=@n13)
   )
  )
  n15 = TextureCoordinate()
  n16 = VectorParameter() {group=HighLight, parameterName=HighLight UV, sortPriority=6, defaultValue=<rgba 2 0.5 0 0.5>}
  n17 = AppendVector(A=@n16.R, B=@n16.R)
  n18 = TextureSampleParameter2D(UVs=Panner(Coordinate=Multiply(A=@n15, B=@n17), Speed=AppendVector(A=@n16.B, B=@n16.B))) {
   group=HighLight
   parameterName=HighLight Tex
   samplerType=SAMPLERTYPE_LinearColor
   texture=/Game/Characters/Player/Male/Medium/Bodies/M_MED_SRDrift/Textures/T_ElectricNoise
  }
  n19 = VectorParameter() {group=HighLight, parameterName=HighLight Refine, defaultValue=<rgba 1.5 10 1 0.5>}
  n20 = OneMinus(None=MaterialFunctionCall(Float2=TextureCoordinate()) {materialFunction=&BreakOutFloat2Components}.G)
  n21 = Multiply(
   A=Multiply(
    A=MaterialFunctionCall(Texture=@n18.R, RGB=@n19.RGB, A=@n19.RGB) {materialFunction=&MF_FX_Refine}.Result
    B=VectorParameter() {group=HighLight, parameterName=HighLight Color, defaultValue=<rgba 1 1 1 1>}.RGB
   )
   B=Saturate(None=Multiply(A=@n20) {constB=5})
  )
  n22 = ScalarParameter() {defaultValue=0.5, group=Galaxy, parameterName=GalaxyTiling}
  n23 = Constant()
  n24 = ScalarParameter() {defaultValue=6, group=Galaxy, parameterName=GalaxyAxisFade}
  n25 = Time()
  n26 = Constant()
  n27 = ScalarParameter() {defaultValue=0.02, group=Galaxy, parameterName=Galaxy Rotation Speed}
  n28 = ScalarParameter() {defaultValue=-0.02, group=Galaxy, parameterName=Galaxy Rotation Speed_1}
  n29 = Multiply(
   A=LinearInterpolate(A=@n25, B=Multiply(A=@n25, B=@n26), Alpha=ScalarParameter() {group=Galaxy, parameterName=Background Rotation})
   B=AppendVector(A=AppendVector(A=@n27, B=@n28), B=ScalarParameter() {defaultValue=0.1, group=Galaxy, parameterName=Galaxy Rotation Speed_2})
  )
  n30 = TextureObjectParameter() {
   group=Galaxy
   parameterName=GalaxyTexture
   texture=/Game/Characters/Player/Male/Medium/Bodies/M_Med_Soldier_04/Skins/Galaxy/Textures/T_FN_Nebula
  }
  n31 = Multiply(
   A=VectorParameter() {group=Galaxy, parameterName=Galaxy Color, defaultValue=<rgba 1 1 1 1>}.RGB
   B=Desaturation(
    None=MaterialFunctionCall!(CameraVector, Rotation Speed=@n29, Texture=@n30, Tiling=@n22, Offset=@n23, AxisFadeContrast=@n24) {materialFunction=/Game/Weapons/WeaponSkins/MaterialFunction/TriplanarCameraVector_Rotated}.Result
    Fraction=ScalarParameter() {group=Galaxy, parameterName=Galaxy Saturation}
   )
  )
  n32 = Desaturation(None=@n31, Fraction=ScalarParameter() {defaultValue=0.5, group=Galaxy, parameterName=Star Saturation})
  n33 = Time()
  n34 = ScalarParameter() {defaultValue=-0.125, group=Galaxy, parameterName=Small Star Rotation speed}
  n35 = TextureObject() {samplerType=SAMPLERTYPE_Masks, texture=/Game/Characters/Player/Male/Medium/Bodies/M_Med_Soldier_04/Skins/Galaxy/Textures/Celestial_T_Stars}
  n36 = ComponentMask(None=MaterialFunctionCall(
    Rotation Speed=Multiply(A=@n33, B=@n34)
    Texture=@n35
    Tiling=ScalarParameter() {defaultValue=5, group=Galaxy, parameterName=Small Star Tiling}
   ) {materialFunction=&TriplanarCameraVector_Rotated}.Result) {b=true}
  n37 = LinearInterpolate(
   A=@n31
   B=Multiply(A=@n32, B=VectorParameter() {group=Galaxy, parameterName=Star Brightness, defaultValue=<rgba 300 300 300 1>}.RGB)
   Alpha=Multiply(A=@n36, B=@n36)
  )
  n38 = Multiply(A=@n32, B=VectorParameter() {group=Galaxy, parameterName=Star Brightness_1, defaultValue=<rgba 950 950 950 1>}.RGB)
  n39 = TextureObject() {samplerType=SAMPLERTYPE_Masks, texture=&Celestial_T_Stars}
  n40 = ComponentMask(None=MaterialFunctionCall!(
    CameraVector
    Texture=@n39
    Tiling=ScalarParameter() {defaultValue=2, group=Galaxy, parameterName=Bright Star Tiling}
    Offset
    AxisFadeContrast
   ) {materialFunction=/Engine/Functions/Engine_MaterialFunctions01/Texturing/TriplanarCameraVector}.Result) {r=true}
  n41 = TextureSample(UVs=Panner(Coordinate=TextureCoordinate() {uTiling=2, vTiling=2}) {speedX=0.133, speedY=0.5}) {
   samplerSource=SSM_Wrap_WorldGroupSettings
   samplerType=SAMPLERTYPE_Masks
   texture=&Celestial_T_Stars
  }
  n42 = Multiply(A=Multiply(A=@n40, B=@n40), B=@n41.G)
  n43 = Add(A=@n21, B=Multiply(A=@n21, B=LinearInterpolate(A=@n37, B=@n38, Alpha=@n42)))
  n44 = ScalarParameter() {group=Galaxy, parameterName=Lerp Galaxy, sortPriority=0}
  n45 = LinearInterpolate(
   A=@n14
   B=Multiply(A=@n14, B=LinearInterpolate(A=@n21, B=@n43, Alpha=@n44))
   Alpha=ScalarParameter() {group=HighLight, parameterName=Lerp HighLight}
  )
  n46 = MaterialFunctionCall() {materialFunction=&LocalPosition}
  n47 = DynamicParameter() {parameterIndex=3, paramNames [X,Y,Z,Param4]}
  MaterialExpressionAppend3Vector_1 = MaterialXAppend3Vector!(A=@n47.X, B=@n47.X, C=@n47.X)
  n48 = ScalarParameter() {defaultValue=9.5, group=Halo, parameterName=Halo B Radius}
  n49 = ScalarParameter() {defaultValue=-0.5, group=Halo, parameterName=Halo B Hardness}
  n50 = Multiply(
   A=Power(
    Base=SphereMask(A=@n46.Local Position, B=@MaterialExpressionAppend3Vector_1, Radius=@n48, Hardness=@n49)
    Exp=ScalarParameter() {defaultValue=1.5, group=Halo, parameterName=Halo A Power}
   ) {constExponent=1.5}
   B=ScalarParameter() {defaultValue=100, group=Halo, parameterName=Halo B Intensity}
  )
  n51 = ScalarParameter() {defaultValue=9.5, group=Halo, parameterName=Halo A Radius}
  n52 = ScalarParameter() {defaultValue=-0.5, group=Halo, parameterName=Halo A Hardness}
  n53 = Multiply(
   A=Power(
    Base=SphereMask(A=@n46.Local Position, B=@MaterialExpressionAppend3Vector_1, Radius=@n51, Hardness=@n52)
    Exp=ScalarParameter() {defaultValue=1.5, group=Halo, parameterName=Halo A Power}
   ) {constExponent=1.5}
   B=ScalarParameter() {defaultValue=100, group=Halo, parameterName=Halo A Intensity}
  )
  n54 = Multiply(A=Multiply(A=Add(A=@n50, B=@n53), B=VectorParameter() {group=Halo, parameterName=Halo Color, defaultValue=<rgba 0 0.8333 1 1>}.RGB), B=@n45)
  n55 = Multiply(
   A=LinearInterpolate(A=@n45, B=Add(A=@n54, B=@n45), Alpha=ScalarParameter() {group=Halo, parameterName=Use Halo})
   B=ScalarParameter() {defaultValue=1, group=Color, parameterName=Emissive Intensity}
  )
  MaterialFunctionCall(Emissive=StaticSwitchParameter(True=Multiply(A=@n1.RGB, B=@n55), False=@n55) {defaultValue=true, parameterName=Use ParticleColor}) {materialFunction=&MF_FX_EmissiveAdjustments}
  n56 = SphereMask(
   A=TextureCoordinate()
   B=Constant2Vector() {g=0.5, r=0.5}
   Radius=ScalarParameter() {defaultValue=0.12, group=SpheraMask, parameterName=UV Radius}
   Hardness=ScalarParameter() {defaultValue=0.5, group=SpheraMask, parameterName=UV Hardness}
  ) {hardnessPercent=0}
  n57 = DynamicParameter() {parameterIndex=3, paramNames [X,Y,Z,Param4]}
  MaterialExpressionAppend3Vector_0 = MaterialXAppend3Vector(A=@n57.X, B=@n57.X, C=@n57.X)
  n58 = SphereMask(
   A=MaterialFunctionCall() {materialFunction=&LocalPosition}.Local Position
   B=@MaterialExpressionAppend3Vector_0
   Radius=ScalarParameter() {defaultValue=10, group=SpheraMask, parameterName=Local Radius}
   Hardness=ScalarParameter() {defaultValue=0.7, group=SpheraMask, parameterName=Local Hardness}
  )
  n59 = LinearInterpolate(
   A=Constant() {r=1}
   B=Multiply(A=OneMinus(None=@n56), B=OneMinus(None=@n58))
   Alpha=ScalarParameter() {defaultValue=1, group=SpheraMask, parameterName=Use SpheraMask}
  ) {constA=1}
  n60 = MaterialFunctionCall(Float2=TextureCoordinate()) {materialFunction=&BreakOutFloat2Components}
  n61 = Power(
   Base=StaticSwitchParameter(True=OneMinus(None=@n60.G), False=@n60.G) {group=ShapeMask, parameterName=Invert ShapeMask B}
   Exp=ScalarParameter() {defaultValue=10, group=ShapeMask, parameterName=ShapeMask B Power}
  )
  n62 = MaterialFunctionCall(Float2=TextureCoordinate()) {materialFunction=&BreakOutFloat2Components}
  n63 = Power(Base=OneMinus(None=@n62.G), Exp=ScalarParameter() {defaultValue=3, group=ShapeMask, parameterName=X Tail Exp}) {constExponent=4}
  n64 = ScalarParameter() {defaultValue=2, group=ShapeMask, parameterName=X Tail Mult}
  n65 = Multiply(A=@n62.G, B=ScalarParameter() {defaultValue=0.47, group=ShapeMask, parameterName=X Offset}) {constB=0.49}
  n66 = Power(Base=OneMinus(None=Sine(None=@n65)), Exp=ScalarParameter() {defaultValue=4, group=ShapeMask, parameterName=X Power}) {constExponent=4}
  n67 = Multiply(A=@n62.R) {constB=0.5}
  n68 = ScalarParameter() {defaultValue=4, group=ShapeMask, parameterName=Y Tail Power}
  n69 = ScalarParameter() {defaultValue=0.5, group=ShapeMask, parameterName=Y Blade Power}
  n70 = OneMinus(None=Power(Base=Sine(None=@n67), Exp=LinearInterpolate(A=@n68, B=@n69, Alpha=@n62.G) {constA=6, constB=2}) {constExponent=4})
  n71 = Add(A=Saturate(None=Multiply(A=@n63, B=@n64)), B=Add(A=Saturate(None=@n66), B=Saturate(None=@n70)))
  n72 = StaticSwitchParameter(True=OneMinus(None=Saturate(None=@n61)), False=@n71) {group=ShapeMask, parameterName=Switch ShapeMask Type}
  n73 = MaterialFunctionCall(Float2=TextureCoordinate()) {materialFunction=&BreakOutFloat2Components}
  n74 = Multiply(A=@n73.G) {constB=0.45}
  n75 = Constant() {r=10}
  n76 = Constant() {r=100}
  n77 = OneMinus(None=@n73.R)
  n78 = Multiply(
   A=Power(Base=Sine(None=@n74), Exp=LinearInterpolate(A=@n75, B=@n76, Alpha=@n73.R) {constA=6, constB=2})
   B=Saturate(None=Power(Base=@n77) {constExponent=4})
  ) {constA=1}
  n79 = Saturate(None=Power(Base=OneMinus(None=@n78)) {constExponent=4})
  n80 = ScalarParameter() {defaultValue=3, group=Noise, parameterName=Noise Power}
  n81 = ParticleRandom()
  n82 = VectorParameter() {group=Distortion, parameterName=Distortion UV, sortPriority=1, defaultValue=<rgba 0.5 2 0 1>}
  n83 = Multiply(A=TextureCoordinate(), B=AppendVector(A=@n82.R, B=@n82.R))
  n84 = AppendVector(A=@n82.B, B=@n82.B)
  n85 = TextureSampleParameter2D(UVs=Add(A=AppendVector(A=@n81, B=@n81), B=Panner(Coordinate=@n83, Speed=@n84))) {
   group=Distortion
   parameterName="Distortion  Tex"
   samplerType=SAMPLERTYPE_Normal
   sortPriority=0
   texture=&T_Weap_Distort_N_02
  }
  n86 = Add(
   A=Multiply(
    A=ComponentMask(None=@n85.RGB) {g=true, r=true}
    B=ScalarParameter() {defaultValue=0.2, group=Distortion, parameterName=Distortion Intensity Noise, sortPriority=2}
   )
   B=TextureCoordinate()
  )
  n87 = VectorParameter() {group=Noise, parameterName=Noise UV, sortPriority=6, defaultValue=<rgba 2 0.5 0 0.7>}
  n88 = AppendVector(A=@n87.R, B=@n87.R)
  n89 = TextureSampleParameter2D(UVs=Panner(Coordinate=Multiply(A=@n86, B=@n88), Speed=AppendVector(A=@n87.B, B=@n87.B))) {
   group=Noise
   parameterName=Noise Tex
   samplerType=SAMPLERTYPE_LinearColor
   texture=&T_ElectricNoise
  }
  n90 = Saturate(None=Multiply(A=Power(Base=@n71, Exp=@n80) {constExponent=4}, B=@n89.R))
  n91 = Multiply(
   A=LinearInterpolate(A=@n72, B=Multiply(A=@n72, B=@n79), Alpha=ScalarParameter() {defaultValue=1, group=ShapeMask, parameterName=Lerp Extra Fade})
   B=LinearInterpolate(B=OneMinus(None=@n90), Alpha=ScalarParameter() {defaultValue=1, group=Noise, parameterName=Lerp Noise}) {constA=1}
  )
  n92 = PixelDepth()
  n93 = Multiply(A=Distance(A=ParticlePositionWS(), B=CameraPositionWS())) {constB=20}
  n94 = ComponentMask(None=SceneTexture() {sceneTextureId=PPI_CustomDepth}.Color) {r=true}
  n95 = Constant() {r=1}
  n96 = Constant() {r=1}
  n97 = FeatureLevelSwitch(
   Default=If(A=Divide(A=@n92.R, B=@n93) {constB=40960}, B=Divide(A=@n94, B=@n93) {constB=40960}, "A > B"=Constant(), "A == B"=@n95, "A < B"=@n95)
   ES2=@n96
   ES3_1=@n96
  )
  n98 = Constant() {r=1}
  n99 = Multiply(
   A=Multiply(A=Multiply(A=@n59, B=@n91), B=StaticSwitchParameter(True=@n97, False=@n98) {group=Backface Culling, parameterName=Use Backface Culling})
   B=ScalarParameter() {defaultValue=1, group=Opacity, parameterName=Global Opacity}
  )
  StaticSwitchParameter(True=Multiply(A=ParticleColor().A, B=@n99), False=@n99) {defaultValue=true, parameterName=Use ParticleColor}
 ]
}
