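# Masked unlit material for Niagara mesh particles that draws the black hole effect of the Inferno pickaxe.
# A panning noise texture, mapped on radial UV1 coordinates by default, is multiplied by a Fresnel term and blended with it to lerp between dark and light colors.
# A second, by default tighter Fresnel adds a noise-modulated colored rim.
# The result, optionally tinted by particle color, feeds a shared emissive adjustment function.
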Material /BRCosmetics/Weapons/FORT_Melee/Pickaxe_Inferno/FX/M_Pickaxe_Inferno_BlackHole
graph {
 properties {
  bUsedWithNiagaraMeshParticles=true
  blendMode=BLEND_Masked
  pixelDepthOffsetMode=PDOM_Legacy
  shadingModel=MSM_Unlit
  usedShadingModels=Unlit
 }
 nodes [
  n1 = ParticleColor()
  n2 = Fresnel(ExponentIn=ScalarParameter() {defaultValue=3, parameterName=Rim ExponentIn}) {baseReflectFraction=0}
  n3 = VectorParameter() {parameterName=Color Rim, defaultValue=<rgba 2.4 0.8 4 1>}
  n4 = Fresnel(ExponentIn=ScalarParameter() {defaultValue=1.8, parameterName=Fresnel ExponentIn}) {baseReflectFraction=0}
  n5 = Subtract(A=TextureCoordinate() {coordinateIndex=1, uTiling=2, vTiling=2}, B=Constant() {r=1})
  n6 = VectorParameter() {parameterName=Noise UV, sortPriority=6, defaultValue=<rgba 3 1 -0.25 1.5>}
  n7 = Multiply(
   A=StaticSwitchParameter(
    True=MaterialFunctionCall(Vector or UVs=@n5) {materialFunction=&VectorToRadialValue}.Radial Coordinates
    False=TextureCoordinate()
   ) {defaultValue=true, parameterName=Use Radia Noise}
   B=AppendVector(A=@n6.R, B=@n6.R)
  )
  n8 = AppendVector(A=@n6.B, B=@n6.B)
  n9 = Multiply(
   A=@n4
   B=TextureSample!(UVs=Panner(Coordinate=@n7, Speed=@n8), Tex, Level, Apply View MipBias) {mipValueMode=TMVM_MipLevel, texture=/Game/Effects/Fort_Effects/Textures/Power_Source/T_Power_Source_Sphere_Texture}.R
  )
  n10 = ScalarParameter() {defaultValue=0.7, parameterName=Lerp Noise}
  n11 = Add(
   A=Multiply(A=Multiply(A=@n2, B=@n3.RGB), B=@n9)
   B=LinearInterpolate(
    A=VectorParameter() {parameterName=Color Dark, defaultValue=<rgba 0 0 0.005 1>}.RGB
    B=VectorParameter() {parameterName=Color Light, defaultValue=<rgba 0.4333 0.15 1 1>}.RGB
    Alpha=LinearInterpolate(A=@n4, B=@n9, Alpha=@n10)
   )
  )
  MaterialFunctionCall(Emissive=StaticSwitchParameter(True=Multiply(A=@n1.RGB, B=@n11), False=@n11) {parameterName=Use ParticleColor}) {materialFunction=&MF_FX_EmissiveAdjustments}
 ]
}
